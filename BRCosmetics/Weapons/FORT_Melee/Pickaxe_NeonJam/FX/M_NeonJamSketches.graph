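# Unlit masked Niagara sprite material for the sketch effects of the NeonJam pickaxe.
# One packed mask texture drives it: the green channel, hard-thresholded at mid value, switches particle color to black before the shared emissive adjustments function.
# The red channel passes through a narrow SmoothStep at particle alpha to form the mask, so alpha acts as an erosion threshold where higher values cut away more of the shape.
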Material /BRCosmetics/Weapons/FORT_Melee/Pickaxe_NeonJam/FX/M_NeonJamSketches
graph {
 properties {bUsedWithNiagaraSprites=true, blendMode=BLEND_Masked, pixelDepthOffsetMode=PDOM_Legacy, shadingModel=MSM_Unlit, usedShadingModels=Unlit}
 nodes [
  n1 = ParticleColor()
  n2 = Constant() {r=0.5}
  n3 = Constant() {r=0.5}
  n4 = TextureSampleParameter2D!(UVs, Bias=Constant(), Apply View MipBias) {mipValueMode=TMVM_MipBias, parameterName=Param, texture=&/T_NeonJamSketches_M}
  MaterialFunctionCall(Emissive=LinearInterpolate(A=@n1.RGB, B=Constant(), Alpha=SmoothStep(Min=@n2, Max=@n3, Value=@n4.G))) {materialFunction=&MF_FX_EmissiveAdjustments}
  SmoothStep(Min=@n1.A, Max=Add(A=@n1.A) {constB=0.01}, Value=@n4.R)
 ]
}
