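# Physics asset for the Neon Jam two-handed pickaxe, giving its handle and crown piece collision bodies and joints, apparently for secondary motion on the crown.
# A capsule covers part of the handle, while small spheres sit on the crown master bone and the offset crown bone.
# The master joint to the handle rotates freely on all axes.
# The crown is angularly locked to the master, so both pivot together as one rigid piece.
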PhysicsAsset /BRCosmetics/Weapons/FORT_Melee/Pickaxe_NeonJam/Meshes/NeonJam_Axe_Physics
bodies {
 Melee_TwoHanded_Handle [{
   shapeType=Capsule
   length=15.2978
   radius=10.1461
   shapeName=Melee_TwoHanded_Handle_capsule
   center=<0.783 -0.996 91.2694>
   rotation=<rot 0 0 87.3431>
  }]
 dyn_crown [{shapeType=Sphere, radius=7.1753, shapeName=dyn_crown_sphere, center=<7.162 -0.0072 -0.1102>}]
 dyn_crown_MSTR [{shapeType=Sphere, radius=0.505, shapeName=dyn_crown_MSTR_sphere}]
}
constraints [
 {
  bone1Name=dyn_crown_MSTR
  bone2Name=Melee_TwoHanded_Handle
  swing1LimitDegrees=45
  swing1Motion=Free
  swing2LimitDegrees=45
  swing2Motion=Free
  twistLimitDegrees=45
  twistMotion=Free
 }
 {
  bone1Name=dyn_crown
  bone2Name=dyn_crown_MSTR
  swing1LimitDegrees=10
  swing1Motion=Locked
  swing2LimitDegrees=45
  swing2Motion=Locked
  twistLimitDegrees=45
  twistMotion=Locked
 }
]
