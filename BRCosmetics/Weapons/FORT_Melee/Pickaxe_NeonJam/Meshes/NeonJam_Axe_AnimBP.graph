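# Animation blueprint for the Neon Jam pickaxe that animates its crown with procedural motion on top of physics.
# A partially blended RigidBody node with overridden gravity feeds two bone modifiers, one bobbing the crown master bone, one rotating the crown.
# The update event drives a looping sine bob and a timed crown roll that eases past a full turn, settles back, then resets.
# It also sets jump-correction alphas from owner vertical velocity, which the AnimGraph never reads.
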AnimBlueprint /BRCosmetics/Weapons/FORT_Melee/Pickaxe_NeonJam/Meshes/NeonJam_Axe_AnimBP
target_skeleton=&/NeonJam_Axe_Skeleton
graphs [
 AnimGraph {nodes [
   AnimGraphNode_SubInput_0 = Input Pose()
   n1 = RigidBody(ComponentPose=Local To Component(LocalPose=@AnimGraphNode_SubInput_0), Alpha=0.65) {
    bOverrideWorldGravity=true
    overridePhysicsAsset=&/NeonJam_Axe_Physics
    alphaBoolBlend {blendOption=Linear}
    componentAppliedLinearAccClamp=<4000 4000 4000>
    componentLinearAccScale=<0.2 0.2 0.2>
    componentLinearVelScale=<0.2 0.2 0.2>
    overrideWorldGravity {z=-980}
   }
   n2 = "Transform (Modify) Bone - Bone: dyn_crown_MSTR"(ComponentPose=@n1, Alpha=1, Translation=Get VerticalMovement() -> (VerticalMovement:Vector), Rotation="0,0,0", Scale="1,1,1") {
    rotationSpace=BCS_BoneSpace
    translationMode=BMM_Additive
    translationSpace=BCS_BoneSpace
    alphaBoolBlend {blendOption=Linear}
    boneToModify {boneName=dyn_crown_MSTR}
   }
   n3 = "Transform (Modify) Bone - Bone: dyn_crown"(ComponentPose=@n2, Alpha=1, Translation="0,0,0", Rotation=Get CrownRotation() -> (CrownRotation:Rotator), Scale="1,1,1") {
    rotationMode=BMM_Additive
    rotationSpace=BCS_BoneSpace
    translationSpace=BCS_BoneSpace
    alphaBoolBlend {blendOption=Linear}
    boneToModify {boneName=dyn_crown}
   }
   Output Pose(Result=Component To Local(ComponentPose=@n3)) {layerGroup=None}
  ]}
 EventGraph {nodes [
   n1 = Event BlueprintUpdateAnimation()
   > n2 = Sequence!(execute) -> (then_0, then_1, then_2:Exec)
   n3 = float * float(A=@n1.DeltaTimeX, B=0.2) {is=title K2Node_PromotableOperator}
   n4 = Get SinTime() -> (SinTime:double)
   > Set SinTime(SinTime:double=Fraction!(A:double=float + float(A=@n4, B=@n3)) -> (ReturnValue:double)) -> (Output_Get:double)
   n5 = float * float(A="Sin (Radians)"(A=Lerp(A=-3.14, B=3.14, Alpha=@n4)), B=1.2)
   > Set VerticalMovement(VerticalMovement:Vector=MakeVector(X=float + float(A=@n5, B=1), Y=0)) -> (Output_Get:Vector)
   n6 = Get IdleTimer() -> (IdleTimer:double)
   Set IdleTimer(execute=@n2.then_1, IdleTimer:double=float + float(A=@n1.DeltaTimeX, B=@n6)) -> (Output_Get:double)
   > Branch(Condition="float >= float"(A=@n6, B=10))
   n7 = float * float(A=1, B=0.2)
   n8 = Get RotationAlpha() -> (RotationAlpha:double)
   > Set RotationAlpha(RotationAlpha:double=float + float(A=float * float(A=@n1.DeltaTimeX, B=@n7), B=@n8)) -> (Output_Get:double)
   > Set!(
    execute
    CrownRotation_Roll:float=FInterpEaseInOut!(A:double=0, B:double=400, Alpha:double=@n8, Exponent:double=3) -> (ReturnValue:double)
    CrownRotation_Pitch:float=0
    CrownRotation_Yaw:float=0
   ) -> (then, Output_Get:Rotator)
   > Branch(Condition="float >= float"(A=@n8, B=1))
   n9 = float * float(A=1, B=0.4)
   n10 = Get InverseRotationAlpha() -> (InverseRotationAlpha:double)
   > Set InverseRotationAlpha(InverseRotationAlpha:double=float + float(A=float * float(A=@n1.DeltaTimeX, B=@n9), B=@n10)) -> (Output_Get:double)
   > Set(CrownRotation_Roll=FInterpEaseInOut(A=400, B=360, Alpha=@n10, Exponent=2), CrownRotation_Pitch=0, CrownRotation_Yaw=0)
   > Branch(Condition="float >= float"(A=@n10, B=1))
   > Set RotationAlpha(RotationAlpha=0)
   > Set CrownRotation(CrownRotation:Rotator="0, 0, 0") -> (Output_Get:Rotator)
   > Set IdleTimer(IdleTimer=0)
   > Set InverseRotationAlpha(InverseRotationAlpha=0)
   n11 = "Not Equal (Float)"(A=GetVelocity!(self:Actor Object Reference=TryGetPawnOwner()) -> (ReturnValue_X:double, ReturnValue_Y:double, ReturnValue_Z:double).ReturnValue_Z)
   Set AnimDynJumpingCorrectionAlpha(execute=@n2.then_2, AnimDynJumpingCorrectionAlpha:double=SelectFloat(A=0, B=0.15, bPickA=@n11)) -> (Output_Get:double)
   > Set RigidBodyJumpingCorrectionAlpha(RigidBodyJumpingCorrectionAlpha:double=SelectFloat(A=0.15, B=0, bPickA=@n11)) -> (Output_Get:double)
  ]}
]
variables [
 AnimDynJumpingCorrectionAlpha {type=number}
 CrownRotation {type=Rotator}
 IdleTimer {type=number, default=2.8148}
 InverseRotationAlpha {type=number}
 Is AnimTime? {type=boolean}
 RigidBodyJumpingCorrectionAlpha {type=number}
 RotationAlpha {type=number}
 SinTime {type=number}
 VerticalMovement {type=Vector}
]
